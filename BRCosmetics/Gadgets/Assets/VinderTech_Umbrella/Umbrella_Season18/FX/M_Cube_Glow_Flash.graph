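# Translucent Niagara particle material, named as a cube glow flash effect for the Season 18 VinderTech umbrella.
# Opposing panned dissolve and blurred noise textures, pulsed by sine waves, pass through narrow equality tests to form animated contour lines.
# These lines modulate a depth-faded mask built from Fresnel, with dynamic-parameter exponent control, and a depth intersection term.
# Particle color feeds an emissive adjustment function, while particle alpha scales the mask and its inverse in separate branches.
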Material /BRCosmetics/Gadgets/Assets/VinderTech_Umbrella/Umbrella_Season18/FX/M_Cube_Glow_Flash
graph {
 properties {
  bUsedWithNiagaraMeshParticles=true
  bUsedWithNiagaraRibbons=true
  bUsedWithNiagaraSprites=true
  blendMode=BLEND_Translucent
  pixelDepthOffsetMode=PDOM_Legacy
  usedShadingModels=Default Lit
 }
 nodes [
  n1 = TextureSample(UVs=Panner() {speedX=-0.5, speedY=-0.5}) {texture=/Game/CharClasses/Outlander/Gadgets/AntiMaterialCharge/Materials/T_AntiMaterialCharge_Dissolve_01}
  n2 = Sine(None=Time() {period=1}) {period=4}
  n3 = Add(A=@n1.R, B=@n2)
  Add(A=@n1.R, B=@n3) {constB=0.5}
  n4 = TextureSample(UVs=Panner() {speedX=-0.2, speedY=-0.2}) {texture=&T_AntiMaterialCharge_Dissolve_01}
  n5 = Sine(None=Time() {period=1}) {period=5}
  n6 = Add(
   A=@n4.R
   B=MaterialFunctionCall!(Sine Phase=@n5, Value 1=Constant() {r=0.2}, Value 2=Constant() {r=0.8}) {materialFunction=/Engine/Functions/Engine_MaterialFunctions02/Utility/Sine_Remapped}.Remapped Sine
  )
  Add(A=@n4.R, B=@n6) {constB=0.5}
  Frac(None=@n5)
  Frac(None=@n2)
  n7 = Panner() {speedX=0.5, speedY=0.5}
  n8 = Add(
   A=TextureSample(UVs=Panner() {speedX=0.3, speedY=0.3}) {samplerType=SAMPLERTYPE_Grayscale, texture=&LowResBlurredNoise}.R
   B=@n6
  )
  n9 = Multiply(
   A=Add(A=TextureSample(UVs=@n7) {samplerType=SAMPLERTYPE_Grayscale, texture=&LowResBlurredNoise}.R, B=@n3)
   B=@n8
  )
  LinearInterpolate(Alpha=@n9)
  LinearInterpolate()
  n10 = ParticleColor()
  MaterialFunctionCall(Emissive=@n10.RGB) {materialFunction=&MF_FX_EmissiveAdjustments}
  n11 = Constant()
  n12 = Constant()
  n13 = Add(
   A=If(A=@n9, B=Constant() {r=1}, "A > B"=@n11, "A == B"=Constant() {r=1}, "A < B"=@n11) {equalsThreshold=0.01}
   B=If(A=@n8, B=Constant() {r=1}, "A > B"=@n12, "A == B"=Constant() {r=1}, "A < B"=@n12) {equalsThreshold=0.05}
  )
  n14 = Fresnel(ExponentIn=Multiply(
    A=DynamicParameter() {paramNames [FresnelMult,Param2,Param3,Param4]}.FresnelMult
    B=ScalarParameter() {defaultValue=1, parameterName=Fresnel Exponent}
   )) {baseReflectFraction=0}
  n15 = LinearInterpolate(Alpha=Power(Base=DepthFade()) {constExponent=0.25}) {constA=1, constB=0}
  n16 = Clamp(None=DepthFade(Opacity=Add(A=@n14, B=@n15)) {fadeDistanceDefault=20})
  Multiply(A=@n10.A, B=Add(A=Multiply(A=@n13, B=@n16), B=@n16))
  Multiply(A=@n10.A, B=OneMinus(None=@n16))
 ]
}
