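# Translucent, lit soft dust ring material for Niagara sprites, ribbons and meshes in the Season 18 umbrella FX.
# A panning dust texture distorts exponential radial gradients, which mask three parameterized colors that are summed and tinted by particle color.
# A time-of-day function picks a dimmed, half-desaturated version for evening and night before emissive adjustment.
# An alpha chain multiplies that color sum by particle alpha, vertex alpha and a contrasted radial gradient fed by the first mask.
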Material /BRCosmetics/Gadgets/Assets/VinderTech_Umbrella/Umbrella_Season18/FX/M_Soft_Dust_Ring3
graph {
 properties {
  bUsedWithNiagaraMeshParticles=true
  bUsedWithNiagaraRibbons=true
  bUsedWithNiagaraSprites=true
  blendMode=BLEND_Translucent
  pixelDepthOffsetMode=PDOM_Legacy
  usedShadingModels=Default Lit
 }
 nodes [
  n1 = ParticleColor()
  n2 = VectorParameter() {parameterName=Color1, defaultValue=<rgba 0.2789 0.0905 0.0529 1>}
  n3 = Panner(Coordinate=TextureCoordinate() {uTiling=0.7, vTiling=0.9}, Speed=Constant2Vector() {g=1}) {speedY=0.25}
  n4 = TextureSample(UVs=@n3) {texture=&/Umbrella_DustPanner_Gold}
  n5 = ComponentMask(None=@n4.RGB) {r=true}
  n6 = Constant() {r=0.5}
  n7 = Multiply(A=@n4.RGB, B=@n6) {constB=0.2}
  n8 = Constant2Vector() {g=0.5, r=0.8}
  n9 = MaterialFunctionCall(
   In=MaterialFunctionCall(UVs=Add(A=@n5, B=@n7), CenterPosition=@n8, Radius=Constant() {r=0.5}, Density=Constant() {r=1}) {materialFunction=&RadialGradientExponential}.RadialGradientExponential
   Contrast=Constant() {r=1}
  ) {materialFunction=&CheapContrast}
  n10 = VectorParameter() {parameterName=Color2, defaultValue=<rgba 0.521 0.1356 0.0704 1>}
  n11 = Multiply(A=ComponentMask(None=TextureCoordinate()) {g=true, r=true}, B=Constant2Vector() {g=1, r=1})
  n12 = ComponentMask(None=@n4.RGB) {r=true}
  n13 = MaterialFunctionCall(UVs=Add(A=@n11, B=Multiply(A=@n12, B=@n6) {constB=0.2}), CenterPosition=@n8, Radius=Constant() {r=0.4}, Density=Constant() {r=0.5}) {materialFunction=&RadialGradientExponential}
  n14 = Add(A=@n11, B=Multiply(A=ComponentMask(None=@n4.RGB) {r=true}, B=@n6) {constB=0.2})
  n15 = Constant2Vector() {g=0.4, r=0.5}
  n16 = Constant() {r=0.8}
  n17 = Constant() {r=1}
  n18 = Add(
   A=Add(A=Multiply(A=@n2.RGB, B=@n9.Result), B=Multiply(A=@n10.RGB, B=@n13.RadialGradientExponential))
   B=Multiply(
    A=VectorParameter() {parameterName=Color3, defaultValue=<rgba 0.6867 0.2502 0.0802 1>}.RGB
    B=MaterialFunctionCall(UVs=@n14, CenterPosition=@n15, Radius=@n16, Density=@n17) {materialFunction=&RadialGradientExponential}.RadialGradientExponential
   )
  )
  n19 = Multiply(A=@n1.RGB, B=@n18)
  n20 = Multiply(A=Desaturation(None=@n19, Fraction=Constant() {r=0.5})) {constB=0.1}
  MaterialFunctionCall(Emissive=MaterialFunctionCall(Color_Morning=@n19, Color_Day=@n19, Color_Evening=@n20, Color_Night=@n20) {materialFunction=&Color_Time_of_Day}.Result) {materialFunction=&MF_FX_EmissiveAdjustments}
  n21 = Multiply(A=@n18) {constB=2}
  n22 = VertexColor()
  n23 = Constant2Vector() {g=0.5, r=0.5}
  n24 = Constant() {r=1}
  n25 = Constant() {r=1}
  Multiply(
   A=Multiply(A=@n1.A, B=Multiply(A=@n21, B=@n22.A) {constA=1})
   B=MaterialFunctionCall(
    In=MaterialFunctionCall(UVs=@n9.Result, CenterPosition=@n23, Radius=@n24, Density=@n25) {materialFunction=&RadialGradientExponential}.RadialGradientExponential
    Contrast=Constant() {r=2}
   ) {materialFunction=&CheapContrast}.Result
  )
  Multiply(B=VertexColor().A) {constA=1}
  n26 = MaterialFunctionCall(Float2=@n3) {materialFunction=&BreakOutFloat2Components}
  Noise!(
   World Position=MaterialFunctionCall(X=@n26.R, Y=@n26.R) {materialFunction=&MakeFloat3}.Result
   FilterWidth
  ) {levelScale=5, levels=2, outputMax=-1, outputMin=1, scale=2.5}
  TextureCoordinate()
  TextureCoordinate()
 ]
}
